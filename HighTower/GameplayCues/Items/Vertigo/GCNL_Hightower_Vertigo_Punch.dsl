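; Looping gameplay cue that drives the Vertigo punch visuals: a wavy stretched mesh from the player's right hand to the flying fist projectile.
; It takes pawn and projectile from cue parameters, or searches the pawn's attached actors on viewing clients.
; A timeline aims the hand-attached mesh and Niagara effect at the projectile through material and particle parameters.
; On hit, it swaps the fist for Niagara, spawns impact FX and runs a distance-scaled wobble/dissolve timeline.
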(event EventOnLoopingStart(EitherVFX-preferthisversion)
  (params (MyTarget :type Actor%20Object%20Reference) (Parameters :type Gameplay%20Cue%20Parameters%20Structure%20%28by%20ref%29) (ParticleComponents :type Array%20of%20FXSystem%20Component%20Object%20References) (AudioComponents :type Array%20of%20Audio%20Component%20Object%20References))
  (bind (_normalizedmagnitude _rawmagnitude _effectcontext _matchedtagname _originaltag _aggregatedsourcetags _aggregatedtargettags _location _normal _instigator _effectcauser _sourceobject) (Ability|GameplayCue|BreakGameplayCueParameters Parameters))
  (bind _playerpawn (Variables|Default|GetPlayerPawn))
  (bind _returnvalue (Customization|GetSkeletalMeshforPartType _playerpawn "Body" :!class /Script/FortniteGame.FortPlayerPawn))
  (bind _projectile (Variables|Default|GetProjectile))
  (bind _returnvalue_2 (Projectile|GetCurrentLocation _projectile))
  (bind _projectile_2 (Variables|Default|GetProjectile))
  (bind _staticmesh (Variables|Default|GetStaticMesh))
  (bind _projectile_3 (Variables|Default|GetProjectile))
  ;[1 1 1] DEBUG: Make sure it's starting every time you activate.  It might not be getting removed properly.
  (Development|PrintString 0 "GC looppp start")
  (Utilities|FlowControl|Sequence
    (:then_0
      ;[1 1 1] Owning client runs through here - everything gets sent correctly
      (bind _asfort_player_pawn (Utilities|Casting|CastToFortPlayerPawn _instigator)
        (:then
          (Variables|Default|SetPlayerPawn _asfort_player_pawn)
          (Utilities|IsValid _sourceobject
            (:Is%20Valid
              (bind _asprj_high_tower_vertigo_punch (Utilities|Casting|CastToPrj_HighTower_Vertigo_Punch _sourceobject)
                (:then
                  (bind _output_get (Variables|Default|SetProjectile _asprj_high_tower_vertigo_punch))
                  (Variables|Default|SetStaticMesh (Variables|/Game/Athena/Items/Consumables/HappyGhost/Prj_Athena_HappyGhost.Prj_Athena_HappyGhost_C|GetStaticMesh _output_get)))
                (:CastFailed
                  (Development|PrintString 0 "Cast ... failed???"))))))
        (:CastFailed
          ;[1 1 1] Viewing clients - they don't get any of the variables for some reason, so we find them by being clever
          (bind _asfort_player_state (Utilities|Casting|CastToFortPlayerState _instigator))
          (bind _asfort_player_pawn_2 (Utilities|Casting|CastToFortPlayerPawn (Variables|/Script/Engine.PlayerState|GetPawnPrivate _asfort_player_state)))
          (Variables|Default|SetPlayerPawn _asfort_player_pawn_2)
          (for _array_element (Actor|GetAttachedActors (Variables|Default|GetPlayerPawn) true false)
            (bind _asprj_high_tower_vertigo_punch_2 (Utilities|Casting|CastToPrj_HighTower_Vertigo_Punch _array_element))
            (Variables|Default|SetProjectile _asprj_high_tower_vertigo_punch_2))
          (Utilities|IsValid _projectile_3
            (:Is%20Valid
              (Variables|Default|SetStaticMesh (Variables|/Game/Athena/Items/Consumables/HappyGhost/Prj_Athena_HappyGhost.Prj_Athena_HappyGhost_C|GetStaticMesh _projectile_3)))))))
    (:then_1
      (Collision|BindEventtoOnActorHit (Variables|Default|GetProjectile) (EventDispatchers|CreateEvent :!event Custom|HitCopy)))
    (:then_2
      (bind _returnvalue_3 (Niagara|SpawnSystemAttached "/HighTower/Items/Vertigo/FX/NS_HighTower_Vertigo_Hand_FX.NS_HighTower_Vertigo_Hand_FX" _returnvalue "hand_r" "0, 0, 0" "0, 0, 0" "KeepRelativeOffset" true))
      (Variables|Default|SetHandParticle _returnvalue_3)
      (Transformation|AttachComponentToComponent (Variables|Default|GetStaticMesh) _returnvalue "hand_r" "KeepRelative" "SnapToTarget" "KeepRelative" true)
      (bind (_direction _sinlength _sinmotion _sinintensity _disslove _disslove_hand) (Timeline|Timeline_0 Play)
        (:Update
          (Utilities|IsValid _projectile_2
            (:Is%20Valid
              (if (not (Delivery|IsBeingKilled _projectile_2))
                (Rendering|Material|SetVectorParameterValueonMaterials _staticmesh "TargetPosition" _returnvalue_2)
                (Rendering|Material|SetScalarParameterValueonMaterials _staticmesh "SinMotion" _sinmotion)
                (if (not (Variables|Default|GetHit))
                  (Rendering|Material|SetScalarParameterValueonMaterials _staticmesh "SinLength" _sinlength)
                  (Rendering|Material|SetScalarParameterValueonMaterials _staticmesh "SinIntensity" (* _sinintensity 5e+01))
                  (Rendering|Material|SetScalarParameterValueonMaterials _staticmesh "Dissolve" _disslove)
                  (Transformation|SetWorldRotation _staticmesh (Math|Rotator|FindLookatRotation (Transformation|GetWorldLocation _staticmesh) _returnvalue_2) false false)
                  (Effects|Components|ParticleSystem|SetVectorParameter (Variables|Default|GetHandParticle) "TargetPosition" _returnvalue_2)
                  (Rendering|Material|SetScalarParameterValueonMaterials (Variables|/HighTower/Items/Vertigo/Prj_HighTower_Vertigo_Punch.Prj_HighTower_Vertigo_Punch_C|GetSM_Hightower_Vertigo_HandAnimation_HandA (Variables|Default|GetProjectile)) "Dissolve" _disslove_hand))))))))))

(event Custom|HitCopy
  (params (SelfActor :type Actor%20Object%20Reference) (OtherActor :type Actor%20Object%20Reference) (NormalImpulse :type Vector) (Hit :type Hit%20Result%20Structure%20%28by%20ref%29))
  (bind (_bblockinghit _binitialoverlap _time _distance _location _impactpoint _normal _impactnormal _physmat _hitactor) (Collision|BreakHitResult Hit))
  (bind _staticmesh (Variables|Default|GetStaticMesh))
  (bind _staticmesh_2 (Variables|Default|GetStaticMesh))
  (bind _projectile (Variables|Default|GetProjectile))
  (Development|PrintString 0 "GC hit event")
  (Variables|Default|SetHit true)
  (Rendering|SetVisibility (Variables|/HighTower/Items/Vertigo/Prj_HighTower_Vertigo_Punch.Prj_HighTower_Vertigo_Punch_C|GetSM_Hightower_Vertigo_HandAnimation_HandA _projectile) false false)
  (Components|Activation|Activate (Variables|/HighTower/Items/Vertigo/Prj_HighTower_Vertigo_Punch.Prj_HighTower_Vertigo_Punch_C|GetNS_HighTower_Vertigo_HandAnimation _projectile) false)
  (bind _asplayer_pawn_athena (Utilities|Casting|CastToPlayerPawn_Athena _hitactor)
    (:then
      (Effects|SpawnFXSystemAttached "/HighTower/Items/Vertigo/FX/NS_HighTower_Vertigo_Impact.NS_HighTower_Vertigo_Impact" (Variables|/Script/Engine.Character|GetMesh _asplayer_pawn_athena) "spine_03"))
    (:CastFailed
      (Niagara|SpawnSystematLocation 0 "/HighTower/Items/Vertigo/FX/NS_HighTower_Vertigo_Impact.NS_HighTower_Vertigo_Impact" _impactpoint)))
  (Development|PrintString 0 "Hit" true true "(R=1.000000,G=0.000000,B=0.142484,A=1.000000)")
  (Rendering|Material|SetScalarParameterValueonMaterials _staticmesh_2 "CosineOffset" (Math|Random|RandomFloat))
  (bind (_direction _sinlength _sinintensity _dissolve) (Timeline|Timeline_1 Play)
    (:Update
      (Utilities|IsValid (Variables|Default|GetProjectile)
        (:Is%20Valid
          (Rendering|Material|SetScalarParameterValueonMaterials _staticmesh "SinLength" _sinlength)
          (Rendering|Material|SetScalarParameterValueonMaterials _staticmesh "SinIntensity" (* _sinintensity (Math|Float|Lerp 0.2 1.0 (Math|Float|Clamp(Float) (/ (Math|Vector|VectorLength (- (Transformation|GetWorldLocation _staticmesh_2) _location)) 3.2e+03)))))
          (Rendering|Material|SetScalarParameterValueonMaterials _staticmesh "Dissolve" _dissolve)
          (Transformation|SetWorldRotation _staticmesh "0, 0, 0" false false)
          (Effects|Components|ParticleSystem|SetVectorParameter (Variables|Default|GetHandParticle) "TargetPosition" (Projectile|GetCurrentLocation (Variables|Default|GetProjectile)))
          (Effects|Components|ParticleSystem|SetFloatParameter (Variables|/HighTower/Items/Vertigo/Prj_HighTower_Vertigo_Punch.Prj_HighTower_Vertigo_Punch_C|GetNS_HighTower_Vertigo_HandAnimation (Variables|Default|GetProjectile)) "Dissolve" _dissolve))))))

(fn ConstructionScript)
